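# Master material for Cube Ninja super-leveling character skins, with optional Tower Sentinel chest-emissive and sock-mask features.
# Custom HLSL 3D Voronoi on noise-curled, pre-skinned positions builds cracked cells and detail shards driving color, normals and emissive, varied by pants and tattoo masks.
# Static switches support baking via WPO UV unwrap and using baked textures, with separate Forward and Mobile shading paths.
# A mask optionally blends in original subsurface-profile textures between character PreFX/PostFX functions.
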Material /BRCosmetics/Characters/Player/Male/Medium/Bodies/M_MED_Cube_Ninja/SuperLeveling/M_CubeNinja_Superleveling_Master
graph {
 properties {
  bUseMaterialAttributes=true
  bUsedWithMeshDeformer=true
  bUsedWithMorphTargets=true
  bUsedWithSkeletalMesh=true
  ditherOpacityMask=true
  pixelDepthOffsetMode=PDOM_Legacy
  shadingModel=MSM_FromMaterialExpression
  subsurfaceProfile=&SS_HeroSkin_02
  usedShadingModels=Default Lit | Subsurface Profile
 }
 nodes [
  n1 = NamedRerouteUsage()
  n2 = Multiply(A=NamedRerouteUsage(), B=OneMinus(None=NamedRerouteUsage()))
  n3 = Multiply(A=@n1, B=@n2)
  n4 = Transform(None=NamedRerouteUsage())
  Add(A=@n3, B=@n4) {constB=0}
  n5 = MaterialFunctionCall() {materialFunction=&BaseFN_CharacterFunction_PreFX}
  n6 = TextureSampleParameter2D() {desc=OriginalTextures, group=OriginalTextures, parameterName=OriginalDiffuse, texture=/Engine/EngineMaterials/DefaultDiffuse}
  n7 = TextureSampleParameter2D() {
   desc=OriginalTextures
   group=OriginalTextures
   parameterName=OriginalSpecularMasks
   samplerType=SAMPLERTYPE_Masks
   texture=&T_Fortnite_Default_S
  }
  n8 = TextureSampleParameter2D() {
   desc=OriginalTextures
   group=OriginalTextures
   parameterName=OriginalNormals
   samplerType=SAMPLERTYPE_Normal
   texture=/Engine/EngineMaterials/DefaultNormal
  }
  n9 = SetMaterialAttributes!(
   MaterialAttributes
   Base Color=@n6.RGB
   Metallic=@n7.G
   Specular=@n7.G
   Roughness=@n7.G
   Normal=@n8.RGB
   Shading Model=ShadingModel() {shadingModel=MSM_SubsurfaceProfile}
  ) {attributeSetTypes [
    #1
    #2
    #3
    #4
    #5
    #6
   ]}
  n10 = NamedRerouteUsage()
  n11 = StaticSwitchParameter(True=Constant(), False=NamedRerouteUsage()) {parameterName=BakeOutput}
  n12 = Constant() {r=1}
  n13 = StaticSwitchParameter(True=LinearInterpolate(A=@n12, B=Constant() {r=0.001}, Alpha=NamedRerouteUsage()), False=@n12) {parameterName=UseTS_ChestEmissive}
  n14 = NamedRerouteUsage()
  n15 = StaticSwitchParameter(True=NamedRerouteUsage(), False=NamedRerouteUsage()) {parameterName=BakeOutput}
  n16 = Constant() {r=1}
  n17 = OneMinus(None=ScalarParameter() {group=TowerSentinelSockMasking, parameterName=HideSock})
  n18 = VertexColor()
  n19 = NamedRerouteUsage()
  n20 = StaticSwitchParameter(True=Constant3Vector() {constant=<rgba 0 0 1 0>}, False=NamedRerouteUsage()) {parameterName=BakeOutput}
  n21 = NamedRerouteUsage()
  n22 = Constant()
  n23 = SetMaterialAttributes!(
   MaterialAttributes
   Base Color=ShadingPathSwitch(Default=StaticSwitchParameter(True=@n10, False=@n11) {parameterName=UseBakedTextures}, Forward=@n10, Mobile=@n10)
   Metallic=@n13
   Specular=@n13
   Roughness=StaticSwitchParameter(True=Constant(), False=NamedRerouteUsage()) {parameterName=BakeOutput}
   Emissive Color=ShadingPathSwitch(Default=StaticSwitchParameter(True=@n14, False=@n15) {parameterName=UseBakedTextures}, Forward=@n14, Mobile=@n14)
   Opacity Mask=StaticSwitchParameter(True=LinearInterpolate(A=@n16, B=@n17, Alpha=@n18.B), False=Constant() {r=1}) {group=TowerSentinel, parameterName=UseSockMask}
   Normal=ShadingPathSwitch(Default=StaticSwitchParameter(True=@n19, False=@n20) {parameterName=UseBakedTextures}, Forward=@n19, Mobile=@n19)
   World Position Offset=StaticSwitchParameter(True=Constant(), False=StaticSwitchParameter(True=@n21, False=@n22) {parameterName=BakeOutput}) {parameterName=UseBakedTextures}
   Shading Model=ShadingModel()
  ) {attributeSetTypes [
    #1
    #2
    #3
    #4
    #7
    #8
    #5
    #9
    #6
   ]}
  n24 = TextureSampleParameter2D() {
   group=OriginalTextures
   parameterName=SuperLevelingMask
   samplerType=SAMPLERTYPE_Masks
   texture=&T_WhiteMask
  }
  MaterialFunctionCall(
   AO=NamedRerouteUsage()
   Distance Blend=@n5.Distance Blend
   Fresnel=@n5.Distance Blend
   Pre Skinned Local Position=@n5.Distance Blend
   Material=StaticSwitchParameter(True=BlendMaterialAttributes(A=@n9, B=@n23, Alpha=@n24.R), False=@n23) {group=OriginalTextures, parameterName=BlendWithNonSuperlevelingTextures.}
  ) {materialFunction=&BaseFN_CharacterFunction_PostFX}
  n25 = Add(A=Transform(None=NamedRerouteUsage()), B=@n3) {constB=0}
  MaterialFunctionCall(
   BaseNormal=Transform(None=Normalize(VectorInput=@n25)) {transformSourceType=TRANSFORMSOURCE_World, transformType=TRANSFORM_Tangent}
   AdditionalNormal=NamedRerouteUsage()
  ) {materialFunction=&BlendAngleCorrectedNormals}
  n26 = Multiply(A=ConstantBiasScale(None=TextureCoordinate()) {bias=-0.5, scale=1}, B=ScalarParameter() {defaultValue=1000, parameterName=Size})
  n27 = Constant()
  n28 = Add(A=Multiply(A=WorldPosition().XYZ) {constB=-1}, B=Add(A=AppendVector(A=@n26, B=@n27), B=ActorPositionWS()))
  Multiply(A=@n28, B=ScalarParameter() {parameterName=Unwrap})
  n29 = NamedRerouteUsage()
  n30 = Multiply(A=Max(A=OneMinus(None=@n29), B=ScalarParameter() {parameterName=TattoosAnodizedEffectIntensity}), B=NamedRerouteUsage())
  Multiply(A=@n30, B=LinearInterpolate(B=NamedRerouteUsage(), Alpha=NamedRerouteUsage()) {constA=1})
  n31 = NamedRerouteUsage()
  n32 = Divide(A=@n31, B=ScalarParameter() {defaultValue=0.1, parameterName=CurlSize})
  n33 = TextureObjectParameter() {group=VolumeTextures, parameterName=CurlNoise, texture=/Engine/EngineResources/DefaultVolumeTexture}
  n34 = Multiply(
   A=ScalarParameter() {parameterName=CurlIntensity}
   B=NamedRerouteDeclaration(None=TextureSample(UVs=@n32, Tex=@n33) {samplerType=SAMPLERTYPE_LinearColor}.RGB) {name=CurlNoise1, nodeColor=<rgba 0.5469 0 1 1>}
  )
  n35 = ScalarParameter() {defaultValue=0.1, parameterName=CurlSize2}
  n36 = Multiply(
   A=ScalarParameter() {parameterName=CurlIntensity2}
   B=TextureSample(UVs=Divide(A=@n31, B=@n35), Tex=@n33) {samplerType=SAMPLERTYPE_LinearColor}.RGB
  )
  n37 = Add(A=@n31, B=VectorParameter() {parameterName=VoronoiOffset}.RGB)
  n38 = LinearInterpolate(
   A=VectorParameter() {parameterName=VoronoiScale, defaultValue=<rgba 0.16 0.16 0.16 0>}.RGB
   B=VectorParameter() {parameterName=VoronoiScalePants, defaultValue=<rgba 0.16 0.16 0.16 0>}.RGB
   Alpha=NamedRerouteUsage()
  )
  MaterialExpressionNamedRerouteDeclaration_5 = NamedRerouteDeclaration(None=Add(A=Add(A=@n34, B=@n36), B=Multiply(A=@n37, B=@n38))) {name=CurledPreskinnedPos, nodeColor=<rgba 0 1 0.6719 1>}
  n39 = NamedRerouteUsage()
  n40 = ScalarParameter() {defaultValue=0.01, parameterName=VoronoiEdgeNoiseSize}
  n41 = Constant()
  n42 = Time()
  n43 = ScalarParameter() {defaultValue=0.1, parameterName=VoronoiEdgeNoiseSpeed}
  n44 = MaterialFunctionCall(Z=Frac(None=Multiply(A=@n42, B=@n43) {constB=0.1})) {materialFunction=&MakeFloat3}
  n45 = TextureSampleParameterVolume!(
   UVs=Subtract(A=Multiply(A=@n39, B=@n40), B=StaticSwitchParameter(True=@n41, False=@n44.Result) {parameterName=BakeOutput}) {constB=0}
   Apply View MipBias
  ) {group=VolumeTextures, parameterName=VoronoiSimple}
  n46 = Power(
   Base=Multiply(A=OneMinus(None=@n45.RGB), B=ScalarParameter() {defaultValue=0.15, parameterName=VoronoiEdgeNoiseIntensity})
   Exp=ScalarParameter() {defaultValue=1.5, parameterName=VooronoiEdgeNoisePow}
  )
  MaterialExpressionAdd_15 = Add(A=ScalarParameter() {defaultValue=-0.09, parameterName=VoronoiEdgeWidth}, B=@n46)
  MaterialExpressionAdd_14 = Add(A=Max(A=ScalarParameter() {defaultValue=0.09, parameterName=VoronoiEdgeSoftness}) {constB=0}, B=@MaterialExpressionAdd_15)
  MaterialExpressionScalarParameter_47 = ScalarParameter() {defaultValue=0.05, parameterName=VoronoiNormalsOffset}
  MaterialExpressionScalarParameter_51 = ScalarParameter() {defaultValue=1, parameterName=VoronoiNormalsIntensity}
  MaterialExpressionAdd_26 = Add(A=ScalarParameter() {defaultValue=0.01, parameterName=VoronoiNormalsEdgeWidth}, B=Multiply(A=@n46) {constB=0.6})
  MaterialExpressionAdd_25 = Add(A=Max(A=ScalarParameter() {defaultValue=0.01, parameterName=VoronoiNormalsEdgeSoftness}) {constB=0}, B=@MaterialExpressionAdd_26)
  n47 = Custom!(
   p=@MaterialExpressionNamedRerouteDeclaration_5
   ss1=@MaterialExpressionAdd_14
   ss2=@MaterialExpressionAdd_15
   nrmOffset=@MaterialExpressionScalarParameter_47
   nrmIntensity=@MaterialExpressionScalarParameter_51
   nrm_ss1=@MaterialExpressionAdd_25
   nrm_ss2=@MaterialExpressionAdd_26
  ) {
   code="struct Functions3\r\n{\r\n\tfloat2 random2 (float2 p)\r\n\t{\r\n\treturn frac(sin(float2(dot(p,float2(127.1,311.7)),dot(p,float2(269.5,183.3))))*43758.5453);\r\n\t}\r\n\t\r\n\tfloat3 random3 (float3 p)\r\n\t{\r\n\treturn frac(sin(float3(dot(p,float3(127.1,311.7,56.01)),dot(p,float3(269.5,183.3,98.1)),dot(p,float3(854.5,238.3,542.1))))*43758.5453);\r\n\t}\r\n\t\r\n\t\r\n\tfloat4 voronoi( float3 st )\r\n\t{\r\n\t\tfloat3 color = float3(0.0,0.0,0.0);\r\n\t\tfloat3 i_st = floor(st);\r\n\t\tfloat3 f_st = frac(st);\r\n\t\r\n\t\tfloat m_dist = 8.;  // minimum distance\r\n\t\tfloat3 m_pt;        // minimum pt\r\n\r\n\t\tfor (int k=-1; k<=1; k++ ) {\r\n\t\t\tfor (int j=-1; j<=1; j++ ) {\r\n\t\t\t\tfor (int i=-1; i<=1; i++ ) {\r\n\t\t\t\t\tfloat3 neighbor = float3(float(i),float(j),float(k));\r\n\t\t\t\t\tfloat3 pt = random3(i_st + neighbor);\r\n\t\t\t\t\t//pt = 0.5 + 0.5*sin(u_time + 6.2831*pt);\r\n\t\t\t\t\tfloat3 diff = neighbor + pt - f_st;\r\n\t\t\t\t\tfloat dist = length(diff);\r\n\r\n\t\t\t\t\tif( dist < m_dist ) {\r\n\t\t\t\t\t\tm_dist = dist;\r\n\t\t\t\t\t\tm_pt = pt;\r\n\t\t\t\t\t}\r\n\t\t\t\t}\r\n\t\t\t}\r\n\t\t}\r\n\t\tcolor += dot(m_pt,float3(.3,.6,.1));\r\n\t\treturn float4(m_pt.xyz,color.x);\t\t\r\n\t}\r\n\t\r\n\r\n\r\n\tfloat voronoiDistance( float3 x )\r\n\t{\r\n\t\tint3 p = int3(floor(x));\r\n\t\tfloat3  f = frac(x);\r\n\r\n\t\tint3 mb;\r\n\t\tfloat3 mr;\r\n\r\n\t\tfloat res = 8.0;\r\n\t\tfor( int k=-1; k<=1; k++ )\r\n\t\t{\r\n\t\t\tfor( int j=-1; j<=1; j++ )\r\n\t\t\t{\r\n\t\t\t\tfor( int i=-1; i<=1; i++ )\r\n\t\t\t\t{\r\n\t\t\t\t\tint3 b = int3(i, j, k);\r\n\t\t\t\t\tfloat3  r = float3(b) + random3(p+b)-f;\r\n\t\t\t\t\tfloat d = dot(r,r);\r\n\r\n\t\t\t\t\tif( d < res )\r\n\t\t\t\t\t{\r\n\t\t\t\t\t\tres = d;\r\n\t\t\t\t\t\tmr = r;\r\n\t\t\t\t\t\tmb = b;\r\n\t\t\t\t\t}\r\n\t\t\t\t}\r\n\t\t\t}\r\n\t\t}\r\n\r\n\t\tres = 8.0;\r\n\t\tfor( int k=-2; k<=2; k++ )\r\n\t\t{\r\n\t\t\tfor( int j=-2; j<=2; j++ )\r\n\t\t\t{\r\n\t\t\t\tfor( int i=-2; i<=2; i++ )\r\n\t\t\t\t{\r\n\t\t\t\t\tint3 b = mb + int3(i, j, k);\r\n\t\t\t\t\tfloat3  r = float3(b) + random3(p+b) - f;\r\n\t\t\t\t\tfloat d = dot(0.5*(mr+r), normalize(r-mr));\r\n\r\n\t\t\t\t\tres = min( res, d );\r\n\t\t\t\t}\r\n\t\t\t}\r\n\t\t}\r\n\t\treturn res;\r\n\t}\r\n\r\n\tfloat3 voronoiNormals (float ss1, float ss2, float3 x, float offset, float intensity)\r\n\t{\r\n\t\tx = x - (offset * 0.5f);\r\n\t\tfloat3 pX = x + float3(offset, 0, 0);\r\n\t\tfloat3 pY = x + float3(0, offset, 0);\r\n\t\tfloat3 pZ = x + float3(0, 0, offset);\r\n\t\t\r\n\t\tfloat dX = voronoiDistance(pX);\r\n\t\tfloat dY = voronoiDistance(pY);\r\n\t\tfloat dZ = voronoiDistance(pZ);\r\n\t\t\r\n\t\tfloat edgeX = smoothstep(ss1,ss2,dX);\r\n\t\tfloat edgeY = smoothstep(ss1,ss2,dY);\r\n\t\tfloat edgeZ = smoothstep(ss1,ss2,dZ);\r\n\t\t\r\n\t\tfloat3 intX = float3((edgeX * intensity), 0, 0);\r\n\t\tfloat3 intY = float3(0, (edgeY * intensity), 0);\r\n\t\tfloat3 intZ = float3(0, 0, (edgeZ * intensity));\r\n\t\t\r\n\t\treturn intX + intY + intZ;\r\n\t}\t\r\n};\r\n\r\n\r\n\r\nFunctions3 f3;\r\n\r\nVoronoiCells =  f3.voronoi(p);\r\n\r\nfloat d = f3.voronoiDistance( p );\r\n\r\n\r\nVoronoiEdges = 1.0 - smoothstep(ss1,ss2,d);\r\n\r\nVoronoiNormalOffsets = f3.voronoiNormals(nrm_ss1, nrm_ss2, p, nrmOffset, nrmIntensity);\r\n\r\nreturn 0;\r\n\r\n"
   description=Voronoi
   additionalOutputs [
    {outputName=VoronoiCells, outputType=CMOT_Float4}
    {outputName=VoronoiEdges, outputType=CMOT_Float1}
    {outputName=VoronoiNormalOffsets, outputType=CMOT_Float3}
   ]
   inputs [
    {
     inputName=p
     input {expression=&self:MaterialExpressionNamedRerouteDeclaration_5}
    }
    {
     inputName=ss1
     input {expression=&self:MaterialExpressionAdd_14}
    }
    {
     inputName=ss2
     input {expression=&self:MaterialExpressionAdd_15}
    }
    {
     inputName=nrmOffset
     input {expression=&self:MaterialExpressionScalarParameter_47}
    }
    {
     inputName=nrmIntensity
     input {expression=&self:MaterialExpressionScalarParameter_51}
    }
    {
     inputName=nrm_ss1
     input {expression=&self:MaterialExpressionAdd_25}
    }
    {
     inputName=nrm_ss2
     input {expression=&self:MaterialExpressionAdd_26}
    }
   ]
  }
  NamedRerouteDeclaration(None=@n47.VoronoiEdges) {name=VoronoiEdges, nodeColor=<rgba 1 0 0.75 1>}
  n48 = TextureSampleParameter2D() {
   parameterName=PantsMask
   samplerType=SAMPLERTYPE_Masks
   texture=/BRCosmetics/Characters/Player/Male/Medium/Bodies/M_MED_Cube_Ninja/Human/Textures/T_M_MED_Cube_Ninja_Human_Body_FX
  }
  NamedRerouteDeclaration(None=ChannelMaskParameter(None=@n48.RGB) {maskChannel=Blue, parameterName=PantsMaskChannel, defaultValue=<rgba 0 0 1 0>}.R) {name=PantsMask, nodeColor=<rgba 0.2489 1 0 1>}
  NamedRerouteDeclaration(None=LinearInterpolate(
    A=ScalarParameter() {defaultValue=0.4, parameterName=Roughness}
    B=ScalarParameter() {defaultValue=0.4, parameterName=RoughnessPants}
    Alpha=NamedRerouteUsage()
   )) {name=RoughnessOut, nodeColor=<rgba 0.4531 0 1 1>}
  NamedRerouteDeclaration(None=@n28) {name=WPO-Unwrap, nodeColor=<rgba 0.5 1 0 1>}
  n49 = TextureSampleParameter2D() {
   parameterName=TattoosMask
   texture=/Game/Characters/Player/Male/Medium/Bodies/M_MED_Cube_Ninja/Textures/T_M_MED_Cube_Ninja_Skin_Emissive_Mask_01
  }
  NamedRerouteDeclaration(None=ChannelMaskParameter(None=@n49.RGB) {maskChannel=Green, parameterName=TattoosMaskChannel, defaultValue=<rgba 0 1 0 0>}.R) {name=TattoosMask, nodeColor=<rgba 1 0.8438 0 1>}
  n50 = VertexNormalWS()
  n51 = StaticBool() {value=true}
  n52 = ScalarParameter() {defaultValue=4, parameterName=EmissiveFresnelPower}
  n53 = NamedRerouteUsage()
  n54 = Multiply(
   A=Multiply(A=OneMinus(None=@n53), B=VectorParameter() {parameterName=CracksEmissive, defaultValue=<rgba 50 0 49.3389 0>}.RGB)
   B=NamedRerouteUsage()
  )
  n55 = NamedRerouteUsage()
  n56 = OneMinus(None=@n55)
  n57 = VectorParameter() {parameterName=TattoosEmissive}
  n58 = Multiply(
   A=StaticSwitchParameter(
    True=Constant() {r=1}
    False=MaterialFunctionCall(Normal Vector=@n50, Invert Fresnel=@n51, Power=@n52) {materialFunction=&Fresnel_Function}.Result
   ) {parameterName=BakeOutput}
   B=Add(A=Multiply(A=@n54, B=@n56), B=Multiply(A=@n55, B=@n57.RGB))
  )
  n59 = StaticSwitchParameter(True=Constant(), False=NamedRerouteUsage()) {parameterName=BakeOutput}
  n60 = NamedRerouteUsage()
  NamedRerouteDeclaration(None=StaticSwitchParameter(True=LinearInterpolate(A=@n58, B=@n59, Alpha=@n60), False=@n58) {parameterName=UseTS_ChestEmissive}) {name=EmissiveOut, nodeColor=<rgba 0.875 1 0 1>}
  n61 = Max(
   A=LinearInterpolate(
    A=ScalarParameter() {defaultValue=1, parameterName=AO_Low}
    B=ScalarParameter() {defaultValue=1, parameterName=AO_High}
    Alpha=NamedRerouteUsage()
   )
   B=ScalarParameter() {parameterName=AO_Min}
  )
  n62 = ScalarParameter() {defaultValue=1, parameterName=AO_Max}
  n63 = VectorParameter() {parameterName=BaseColor, defaultValue=<rgba 0.2448 0.2448 0.2448 0>}
  n64 = VectorParameter() {parameterName=BaseColorPants, defaultValue=<rgba 0.6791 1 0.9298 0>}
  n65 = NamedRerouteUsage()
  n66 = ScalarParameter() {defaultValue=0.2, parameterName=BaseColorLow}
  n67 = ScalarParameter() {defaultValue=0.3, parameterName=BaseColorHigh}
  n68 = NamedRerouteUsage()
  n69 = NamedRerouteUsage()
  n70 = ScalarParameter() {parameterName=DisplayCells}
  n71 = Saturate(None=Add(A=LinearInterpolate(B=@n68, Alpha=@n69) {constA=1}, B=OneMinus(None=@n70)))
  n72 = ScalarParameter() {defaultValue=1, parameterName=DisplayShards}
  n73 = Saturate(None=Add(A=NamedRerouteUsage(), B=OneMinus(None=@n72)))
  n74 = NamedRerouteUsage()
  n75 = OneMinus(None=ScalarParameter() {defaultValue=1, parameterName=DisplayCellEdges})
  n76 = Saturate(None=LinearInterpolate(B=Add(A=@n74, B=@n75), Alpha=NamedRerouteUsage()) {constA=1})
  n77 = Multiply(
   A=Multiply(A=Min(A=@n61, B=@n62), B=LinearInterpolate(A=@n63.RGB, B=@n64.RGB, Alpha=@n65))
   B=Multiply(A=LinearInterpolate(A=@n66, B=@n67, Alpha=@n71), B=Multiply(A=@n73, B=@n76))
  ) {constA=1}
  n78 = Multiply(A=@n77, B=VectorParameter() {parameterName=TattoosBaseColorMult, defaultValue=<rgba 0.5677 0.5677 0.5677 0>}.RGB)
  n79 = NamedRerouteUsage()
  n80 = Saturate(None=Add(A=NamedRerouteUsage(), B=ScalarParameter() {parameterName=VoronoiEdgeBaseColorLevel}) {constB=0.4})
  n81 = Max(A=NamedRerouteUsage(), B=OneMinus(None=NamedRerouteUsage()))
  n82 = Max(A=OneMinus(None=NamedRerouteUsage()), B=ScalarParameter() {parameterName=TattoosAnodizedEffectIntensity})
  n83 = NamedRerouteUsage()
  n84 = NamedRerouteUsage()
  n85 = NamedRerouteUsage()
  n86 = Add(
   A=Multiply(A=LinearInterpolate(A=@n77, B=@n78, Alpha=@n79) {constB=0}, B=LinearInterpolate(A=@n80, Alpha=@n81))
   B=Multiply(A=Multiply(A=@n82, B=@n83), B=LinearInterpolate(B=@n84, Alpha=@n85) {constA=1})
  )
  n87 = Constant() {r=0.001}
  n88 = NamedRerouteUsage()
  NamedRerouteDeclaration(None=StaticSwitchParameter(True=LinearInterpolate(A=@n86, B=@n87, Alpha=@n88), False=@n86) {parameterName=UseTS_ChestEmissive}) {name=BaseColorOut, nodeColor=<rgba 0.8047 1 0 1>}
  n89 = Constant() {r=1}
  n90 = Constant() {r=3}
  n91 = PixelNormalWS()
  n92 = Multiply(
   A=TextureSample(UVs=Fresnel(ExponentIn=@n89, BaseReflectFractionIn=@n90, Normal=@n91) {baseReflectFraction=0.5, exponent=3}) {texture=/Game/Global/Textures/Default/Masks_and_Gradients/3pxRGBImage}.RGB
   B=VectorParameter() {group=TowerSentinel_Metal, parameterName=AnodizationFilter, defaultValue=<rgba 1 1 1 0>}.RGB
  )
  n93 = Power(Base=DotProduct(A=AtmosphericLightVector(), B=ReflectionVectorWS()), Exp=Constant() {r=55})
  n94 = Constant() {r=512}
  n95 = LinearInterpolate(A=@n92, B=Multiply(A=Constant() {r=4}, B=@n92), Alpha=Saturate(None=Multiply(A=@n93, B=@n94)))
  n96 = ScalarParameter() {parameterName=AnodizedEffectIntensity}
  NamedRerouteDeclaration(None=StaticSwitchParameter(True=Constant(), False=Multiply(A=@n95, B=@n96) {constB=0.1}) {parameterName=BakeOutput}) {name=AnodizedEffect, nodeColor=<rgba 0 0.9531 1 1>}
  NamedRerouteDeclaration(None=TextureSampleParameter2D() {parameterName=NormalMap, samplerType=SAMPLERTYPE_Normal, texture=&FlatNormalBlue}.RGB) {name=NormalMap, nodeColor=<rgba 1 0 0.4687 1>}
  n97 = PreSkinnedPosition()
  n98 = VectorParameter() {parameterName=Min, defaultValue=<rgba -59.5 -18 0 1>}
  NamedRerouteDeclaration(None=Divide(A=Subtract(A=@n97, B=@n98.RGB), B=VectorParameter() {parameterName=Extends, defaultValue=<rgba 119 36 143 1>}.RGB)) {name=LocalSpaceGradients, nodeColor=<rgba 1 0 0.9141 1>}
  n99 = Multiply(A=Normalize(VectorInput=NamedRerouteUsage())) {constB=0.5}
  n100 = Add(
   A=StaticSwitchParameter(True=NamedRerouteUsage(), False=Constant()) {defaultValue=true, parameterName=BakeBaseColor}
   B=StaticSwitchParameter(True=Add(A=@n99) {constB=0.5}, False=Constant()) {parameterName=BakeNormals}
  )
  n101 = StaticSwitchParameter(
   True=Divide(A=NamedRerouteUsage(), B=ScalarParameter() {defaultValue=1, parameterName=BakedEmissiveMultiplier}) {constB=30}
   False=Constant()
  ) {parameterName=BakeEmissive}
  n102 = NamedRerouteUsage()
  n103 = Constant()
  NamedRerouteDeclaration(None=Add(A=Add(A=@n100, B=@n101), B=StaticSwitchParameter(True=@n102, False=@n103) {parameterName=BakeRoughness})) {name=BakeOutput, nodeColor=<rgba 0 1 0.9297 1>}
  n104 = Subtract(
   A=VectorParameter() {group=TowerSentinel_RadialUV, parameterName=RadialUVOffset2, defaultValue=<rgba 0 0 130 0>}.RGB
   B=VertexInterpolator(VS=PreSkinnedPosition()).PS
  )
  n105 = Divide(A=@n104, B=PreSkinnedLocalBounds().Extents)
  n106 = Divide(A=Arctangent2Fast(Y=ComponentMask(None=@n105) {b=true}, X=ComponentMask(None=@n105) {r=true}), B=Constant() {r=3.1416})
  n107 = Distance(A=ComponentMask(None=@n104) {b=true, r=true}, B=Constant())
  n108 = ObjectRadius()
  n109 = AppendVector(A=Multiply(A=Add(A=@n106) {constB=-0.5}) {constB=0.5}, B=OneMinus(None=Divide(A=@n107, B=@n108)))
  n110 = ComponentMask(None=VectorParameter() {group=TowerSentinel_RadialUV, parameterName=UV_Mul, defaultValue=<rgba 1 1 0 0>}.RGB) {g=true, r=true}
  NamedRerouteDeclaration(None=Frac(None=Multiply(A=@n109, B=@n110))) {name=RadialUV, nodeColor=<rgba 1 0.2812 0 1>}
  n111 = Distance(A=@n47.VoronoiNormalOffsets, B=Constant())
  NamedRerouteDeclaration(None=Saturate(None=OneMinus(None=@n111))) {name=VoronoiNormalLengthInv, nodeColor=<rgba 1 0 0.3516 1>}
  n112 = Add(A=PreSkinnedPosition(), B=NamedRerouteUsage())
  n113 = Constant3Vector() {constant=<rgba 1 0 0 1>}
  n114 = ComponentMask(None=@n113) {g=true, r=true}
  n115 = Constant()
  n116 = NamedRerouteUsage()
  n117 = ScalarParameter() {defaultValue=-1, parameterName=ShardsRotation}
  n118 = ScalarParameter() {defaultValue=-1, parameterName=ShardsRotationPants}
  n119 = NamedRerouteUsage()
  n120 = Frac(None=Multiply(A=ComponentMask(None=@n116) {b=true}, B=LinearInterpolate(A=@n117, B=@n118, Alpha=@n119)))
  n121 = Normalize(VectorInput=AppendVector(
    A=MaterialFunctionCall(UVs=@n114, Rotation Center=@n115, "Rotation Angle (0-1)"=@n120) {materialFunction=&CustomRotator}.Rotated Values
    B=ComponentMask(None=@n113) {b=true}
   ))
  n122 = Constant3Vector() {constant=<rgba 0 0 1 1>}
  n123 = Multiply(A=CrossProduct(A=@n121, B=@n122)) {constB=-1}
  NamedRerouteDeclaration(None=VertexInterpolator(VS=MaterialFunctionCall(VectorToTransform=@n112, BasisX=@n121, BasisY=@n123, BasisZ=@n122) {materialFunction=&Transform3x3Matrix}.Transformed Vector).PS) {name=PreskinnedPosRotated, nodeColor=<rgba 1 0 0.7734 1>}
  NamedRerouteDeclaration(None=VertexInterpolator(VS=PreSkinnedPosition()).PS) {name=PreskinnedPos, nodeColor=<rgba 0 1 0.9766 1>}
  n124 = TextureSampleParameter2D() {
   parameterName=EmissiveCracksMask
   samplerType=SAMPLERTYPE_Masks
   texture=/BRCosmetics/Characters/Player/Male/Medium/Bodies/M_MED_Cube_Ninja/SuperLeveling/Textures/T_M_MED_Cube_Ninja_Human_Body_T1_FX
  }
  n125 = ChannelMaskParameter(None=@n124.RGB) {maskChannel=Green, parameterName=EmissiveCracksMaskChannel, defaultValue=<rgba 0 1 0 0>}
  NamedRerouteDeclaration(None=StaticSwitchParameter(True=OneMinus(None=@n125.R), False=@n125.R) {parameterName=InvertEmissiveCracksMask}) {name=EmissiveCracksMask, nodeColor=<rgba 0.2489 1 0 1>}
  n126 = Add(A=PreSkinnedPosition(), B=NamedRerouteUsage())
  n127 = Constant3Vector() {constant=<rgba 1 0 0 1>}
  n128 = ComponentMask(None=@n127) {g=true, r=true}
  n129 = Constant()
  n130 = NamedRerouteUsage()
  n131 = ScalarParameter() {parameterName=CracksRotation}
  n132 = ScalarParameter() {parameterName=CracksRotationPants}
  n133 = NamedRerouteUsage()
  n134 = Frac(None=Multiply(A=ComponentMask(None=@n130) {b=true}, B=LinearInterpolate(A=@n131, B=@n132, Alpha=@n133)))
  n135 = Normalize(VectorInput=AppendVector(
    A=MaterialFunctionCall(UVs=@n128, Rotation Center=@n129, "Rotation Angle (0-1)"=@n134) {materialFunction=&CustomRotator}.Rotated Values
    B=ComponentMask(None=@n127) {b=true}
   ))
  n136 = Constant3Vector() {constant=<rgba 0 0 1 1>}
  n137 = Multiply(A=CrossProduct(A=@n135, B=@n136)) {constB=-1}
  NamedRerouteDeclaration(None=VertexInterpolator(VS=MaterialFunctionCall(VectorToTransform=@n126, BasisX=@n135, BasisY=@n137, BasisZ=@n136) {materialFunction=&Transform3x3Matrix}.Transformed Vector).PS) {name=PreskinnedPosRotated2, nodeColor=<rgba 1 0 0.7734 1>}
  n138 = PreSkinnedPosition()
  n139 = AppendVector(A=Abs(None=ComponentMask(None=@n138) {r=true}), B=ComponentMask(None=@n138) {b=true, g=true})
  n140 = ShaderStageSwitch(PixelShader=VertexInterpolator(VS=@n139).PS, VertexShader=@n139)
  n141 = Max(
   A=SphereMask(
    A=@n140
    B=VectorParameter() {group=arms rotation fix, parameterName=upperArm, defaultValue=<rgba 31.7102 0 135.5019 0>}.RGB
    Radius=ScalarParameter() {defaultValue=23.4406, group=arms rotation fix, parameterName=radUpperArm}
    Hardness=ScalarParameter() {defaultValue=0.7252, group=arms rotation fix, parameterName=hardUpperArm}
   )
   B=SphereMask(
    A=@n140
    B=VectorParameter() {group=arms rotation fix, parameterName=lowerArm, defaultValue=<rgba 47.3464 0 102.6949 0>}.RGB
    Radius=ScalarParameter() {defaultValue=25.003, group=arms rotation fix, parameterName=radLowerArm}
    Hardness=ScalarParameter() {defaultValue=1, group=arms rotation fix, parameterName=hardLowerArm}
   )
  )
  n142 = Sign(None=ComponentMask(None=PreSkinnedPosition()) {r=true})
  n143 = Multiply(A=ScalarParameter() {defaultValue=-0.4, group=arms rotation fix, parameterName=Angle}, B=@n142)
  n144 = VectorParameter() {group=arms rotation fix, parameterName=shoulder, defaultValue=<rgba 13.8458 0 133.4372 0>}
  n145 = Multiply(A=ComponentMask(None=@n144.RGB) {r=true}, B=@n142)
  n146 = ComponentMask(None=@n144.RGB) {b=true, g=true}
  NamedRerouteDeclaration(None=RotateAboutAxis(
    NormalizedRotationAxis=Constant3Vector() {constant=<rgba 0 1 0 0>}
    RotationAngle=Multiply(A=@n141, B=@n143)
    PivotPoint=AppendVector(A=@n145, B=@n146)
    Position=PreSkinnedPosition()
   )) {name=ArmsRotationOffset, nodeColor=<rgba 0.4297 0 1 1>}
  n147 = Add(
   A=ComponentMask(None=@n104) {b=true, r=true}
   B=ComponentMask(None=VectorParameter() {group=TowerSentinel, parameterName=maskOffset, defaultValue=<rgba 0 -5 0 0>}.RGB) {g=true, r=true}
  )
  n148 = ComponentMask(None=VectorParameter() {group=TowerSentinel, parameterName=maskScale, defaultValue=<rgba 2.2 1 0 0>}.RGB) {g=true, r=true}
  n149 = Divide(A=Distance(A=Multiply(A=@n147, B=@n148), B=Constant()), B=ObjectRadius())
  n150 = Multiply(
   A=Power(Base=OneMinus(None=@n149), Exp=ScalarParameter() {defaultValue=35, group=TowerSentinel, parameterName=maskPow})
   B=ScalarParameter() {defaultValue=6, parameterName=maskMul}
  )
  n151 = Add(A=NamedRerouteUsage(), B=Constant2Vector() {r=0.5})
  n152 = ComponentMask(None=NamedRerouteUsage()) {r=true}
  n153 = Multiply(A=Power(Base=Sine(None=@n152) {period=2})) {constB=5}
  n154 = LinearInterpolate(
   A=MaterialFunctionCall!(In=Frac(None=@n151)) {materialFunction=/Game/Characters/Player/Female/Medium/Bodies/F_MED_Tower_Sentinel/Materials/MF_TowerSentinelChestEmissive}.Result
   B=MaterialFunctionCall(In=NamedRerouteUsage()) {materialFunction=&MF_TowerSentinelChestEmissive}.Result
   Alpha=Saturate(None=Saturate(None=@n153))
  )
  n155 = Desaturation(
   None=MaterialFunctionCall(Hue Shift Percentage=ScalarParameter() {group=TowerSentinel, parameterName=TS_HueShift}, Texture=Multiply(A=@n150, B=@n154)) {materialFunction=&HueShift}.Result
   Fraction=ScalarParameter() {group=TowerSentinel, parameterName=TS_Desat}
  )
  n156 = ScalarParameter() {defaultValue=1, group=TowerSentinel, parameterName=TS_Pow}
  NamedRerouteDeclaration(None=Multiply(A=Power(Base=@n155, Exp=@n156), B=ScalarParameter() {defaultValue=1, group=TowerSentinel, parameterName=TS_Mul})) {name=TS_ChestEmissive, nodeColor=<rgba 0.1953 1 0 1>}
  n157 = TextureSampleParameter2D() {
   parameterName=TS_ChestEmissiveMask
   samplerType=SAMPLERTYPE_Masks
   texture=/Game/Characters/Player/Female/Medium/Bodies/F_MED_Tower_Sentinel/Textures/T_F_MED_Tower_Sentinel_Body_ChestMask
  }
  NamedRerouteDeclaration(None=@n157.R) {name=TS_ChestMask, nodeColor=<rgba 1 0 0.8672 1>}
  NamedRerouteDeclaration(None=ComponentMask(None=@n47.VoronoiCells) {b=true, g=true, r=true}) {name=VoronoiCellColors, nodeColor=<rgba 0.8984 0 1 1>}
  n158 = Add(A=TextureSampleParameter2D() {parameterName=Baked_Diffuse}.RGB, B=@n30)
  n159 = Constant() {r=0.001}
  n160 = NamedRerouteUsage()
  NamedRerouteDeclaration(None=StaticSwitchParameter(True=LinearInterpolate(A=@n158, B=@n159, Alpha=@n160), False=@n158) {parameterName=UseTS_ChestEmissive}) {name=BakedDiffuse, nodeColor=<rgba 1 0.7734 0 1>}
  NamedRerouteDeclaration(None=TextureSampleParameter2D() {parameterName=Baked_NormalMap, samplerType=SAMPLERTYPE_Normal, texture=&FlatNormalBlue}.RGB) {name=BakedNormal, nodeColor=<rgba 1 0.6328 0 1>}
  NamedRerouteDeclaration(None=@n47.VoronoiNormalOffsets) {name=VoronoiNormalWS, nodeColor=<rgba 0.4062 0 1 1>}
  n161 = ComponentMask(None=VectorParameter() {group=BakedScrollingEmissive, parameterName=EmissiveSphereMask, defaultValue=<rgba 0 0 90 0>}.RGB) {b=true, g=true, r=true}
  n162 = NamedRerouteUsage()
  n163 = ScalarParameter() {group=BakedScrollingEmissive, parameterName=EmissiveSphereMaskRadius}
  n164 = Constant() {r=0.5}
  n165 = NamedRerouteUsage()
  n166 = VectorParameter() {group=BakedScrollingEmissive, parameterName=EmMaskScrollingSpeed, defaultValue=<rgba 0 -0.12 0 0>}
  n167 = Add(
   A=ComponentMask(None=VertexInterpolator(VS=@n165).PS) {b=true, r=true}
   B=Multiply(A=ComponentMask(None=@n166.RGB) {g=true, r=true}, B=Time())
  )
  n168 = ScalarParameter() {defaultValue=1, group=BakedScrollingEmissive, parameterName=EmMaskTile}
  n169 = TextureSample(UVs=Frac(None=Multiply(A=@n167, B=@n168))) {samplerType=SAMPLERTYPE_LinearGrayscale, texture=/Game/Packages/Fortress_Sky/SkyDome/Master/T_SkyDome_Cloud_Voroni}
  n170 = ScalarParameter() {defaultValue=3, group=BakedScrollingEmissive, parameterName=EmMaskPower}
  n171 = Multiply(A=OneMinus(None=SphereMask(A=@n161, B=@n162, Radius=@n163, Hardness=@n164)), B=Add(A=Power(Base=@n169.R, Exp=@n170), B=NamedRerouteUsage()))
  n172 = Multiply(A=TextureSampleParameter2D() {parameterName=Baked_Emissive}.RGB, B=ScalarParameter() {defaultValue=1, parameterName=BakedEmissiveMultiplier})
  n173 = ScalarParameter() {defaultValue=2, group=BakedScrollingEmissive, parameterName=EmMultiplier}
  n174 = PixelNormalWS()
  n175 = StaticBool() {value=true}
  n176 = ScalarParameter() {defaultValue=4, parameterName=EmissiveFresnelPower}
  n177 = TextureSampleParameter2D() {
   group=OriginalTextures
   parameterName=SuperLevelingMask
   samplerType=SAMPLERTYPE_Masks
   texture=&T_WhiteMask
  }
  n178 = Constant() {r=1}
  n179 = Multiply(
   A=Multiply(A=Saturate(None=@n171), B=Multiply(A=@n172, B=@n173))
   B=Multiply(
    A=MaterialFunctionCall(Normal Vector=@n174, Invert Fresnel=@n175, Power=@n176) {materialFunction=&Fresnel_Function}.Result
    B=StaticSwitchParameter(True=@n177.B, False=@n178) {parameterName=UseBakedEmissiveMask}
   )
  )
  n180 = NamedRerouteUsage()
  n181 = NamedRerouteUsage()
  NamedRerouteDeclaration(None=StaticSwitchParameter(True=LinearInterpolate(A=@n179, B=@n180, Alpha=@n181), False=@n179) {parameterName=UseTS_ChestEmissive}) {name=BakedEmissive, nodeColor=<rgba 1 0 0.6094 1>}
  NamedRerouteDeclaration(None=ChannelMaskParameter(None=TextureSampleParameter2D() {parameterName=AO, samplerType=SAMPLERTYPE_Masks, texture=&T_WhiteMask}.RGB) {parameterName=AOChannel}.R) {name=AO, nodeColor=<rgba 1 0.375 0 1>}
  NamedRerouteDeclaration(None=ComponentMask(None=@n47.VoronoiCells) {a=true}) {name=VoronoiCells, nodeColor=<rgba 0 0.625 1 1>}
  n182 = NamedRerouteUsage()
  n183 = Constant3Vector() {constant=<rgba 0 0 1 0>}
  n184 = NamedRerouteUsage()
  n185 = LinearInterpolate(
   A=ScalarParameter() {parameterName=DetailVoronoiPerCellRotation}
   B=ScalarParameter() {parameterName=DetailVoronoiPerCellRotationPants}
   Alpha=NamedRerouteUsage()
  )
  MaterialExpressionMultiply_49 = Multiply(
   A=MaterialFunctionCall!(VectorToRotate=NamedRerouteUsage(), LookAtVector=LinearInterpolate(A=@n183, B=@n184, Alpha=@n185), RestingVector) {materialFunction=/Engine/Functions/Engine_MaterialFunctions01/Vectors/RotateVector}.Result
   B=LinearInterpolate(
    A=VectorParameter() {parameterName=DetailVoronoiScale, defaultValue=<rgba 0.3 0.3 0.3 0>}.RGB
    B=VectorParameter() {parameterName=DetailVoronoiScalePants, defaultValue=<rgba 0.3 0.3 0.3 0>}.RGB
    Alpha=NamedRerouteUsage()
   )
  )
  MaterialExpressionScalarParameter_54 = ScalarParameter() {defaultValue=0.01, parameterName=DetailVoronoiEdgeWidth}
  MaterialExpressionAdd_29 = Add(A=Max(A=ScalarParameter() {defaultValue=0.01, parameterName=DetailVoronoiEdgeSoftness}) {constB=0}, B=@MaterialExpressionScalarParameter_54)
  MaterialExpressionScalarParameter_56 = ScalarParameter() {defaultValue=0.031, parameterName=DetailVoronoiNrmOffset}
  MaterialExpressionScalarParameter_55 = ScalarParameter() {parameterName=DetailVoronoiNrmIntensity}
  n186 = Custom(
   p=@MaterialExpressionMultiply_49
   ss1=@MaterialExpressionAdd_29
   ss2=@MaterialExpressionScalarParameter_54
   nrmOffset=@MaterialExpressionScalarParameter_56
   nrmIntensity=@MaterialExpressionScalarParameter_55
   nrm_ss1=@MaterialExpressionAdd_29
   nrm_ss2=@MaterialExpressionScalarParameter_54
  ) {
   code="struct Functions3\r\n{\r\n\tfloat2 random2 (float2 p)\r\n\t{\r\n\treturn frac(sin(float2(dot(p,float2(127.1,311.7)),dot(p,float2(269.5,183.3))))*43758.5453);\r\n\t}\r\n\t\r\n\tfloat3 random3 (float3 p)\r\n\t{\r\n\treturn frac(sin(float3(dot(p,float3(127.1,311.7,56.01)),dot(p,float3(269.5,183.3,98.1)),dot(p,float3(854.5,238.3,542.1))))*43758.5453);\r\n\t}\r\n\t\r\n\t\r\n\tfloat4 voronoi( float3 st )\r\n\t{\r\n\t\tfloat3 color = float3(0.0,0.0,0.0);\r\n\t\tfloat3 i_st = floor(st);\r\n\t\tfloat3 f_st = frac(st);\r\n\t\r\n\t\tfloat m_dist = 8.;  // minimum distance\r\n\t\tfloat3 m_pt;        // minimum pt\r\n\r\n\t\tfor (int k=-1; k<=1; k++ ) {\r\n\t\t\tfor (int j=-1; j<=1; j++ ) {\r\n\t\t\t\tfor (int i=-1; i<=1; i++ ) {\r\n\t\t\t\t\tfloat3 neighbor = float3(float(i),float(j),float(k));\r\n\t\t\t\t\tfloat3 pt = random3(i_st + neighbor);\r\n\t\t\t\t\t//pt = 0.5 + 0.5*sin(u_time + 6.2831*pt);\r\n\t\t\t\t\tfloat3 diff = neighbor + pt - f_st;\r\n\t\t\t\t\tfloat dist = length(diff);\r\n\r\n\t\t\t\t\tif( dist < m_dist ) {\r\n\t\t\t\t\t\tm_dist = dist;\r\n\t\t\t\t\t\tm_pt = pt;\r\n\t\t\t\t\t}\r\n\t\t\t\t}\r\n\t\t\t}\r\n\t\t}\r\n\t\tcolor += dot(m_pt,float3(.3,.6,.1));\r\n\t\treturn float4(m_pt.xyz,color.x);\t\t\r\n\t}\r\n\t\r\n\r\n\r\n\tfloat voronoiDistance( float3 x )\r\n\t{\r\n\t\tint3 p = int3(floor(x));\r\n\t\tfloat3  f = frac(x);\r\n\r\n\t\tint3 mb;\r\n\t\tfloat3 mr;\r\n\r\n\t\tfloat res = 8.0;\r\n\t\tfor( int k=-1; k<=1; k++ )\r\n\t\t{\r\n\t\t\tfor( int j=-1; j<=1; j++ )\r\n\t\t\t{\r\n\t\t\t\tfor( int i=-1; i<=1; i++ )\r\n\t\t\t\t{\r\n\t\t\t\t\tint3 b = int3(i, j, k);\r\n\t\t\t\t\tfloat3  r = float3(b) + random3(p+b)-f;\r\n\t\t\t\t\tfloat d = dot(r,r);\r\n\r\n\t\t\t\t\tif( d < res )\r\n\t\t\t\t\t{\r\n\t\t\t\t\t\tres = d;\r\n\t\t\t\t\t\tmr = r;\r\n\t\t\t\t\t\tmb = b;\r\n\t\t\t\t\t}\r\n\t\t\t\t}\r\n\t\t\t}\r\n\t\t}\r\n\r\n\t\tres = 8.0;\r\n\t\tfor( int k=-2; k<=2; k++ )\r\n\t\t{\r\n\t\t\tfor( int j=-2; j<=2; j++ )\r\n\t\t\t{\r\n\t\t\t\tfor( int i=-2; i<=2; i++ )\r\n\t\t\t\t{\r\n\t\t\t\t\tint3 b = mb + int3(i, j, k);\r\n\t\t\t\t\tfloat3  r = float3(b) + random3(p+b) - f;\r\n\t\t\t\t\tfloat d = dot(0.5*(mr+r), normalize(r-mr));\r\n\r\n\t\t\t\t\tres = min( res, d );\r\n\t\t\t\t}\r\n\t\t\t}\r\n\t\t}\r\n\t\treturn res;\r\n\t}\r\n\r\n\tfloat3 voronoiNormals (float ss1, float ss2, float3 x, float offset, float intensity)\r\n\t{\r\n\t\tx = x - (offset * 0.5f);\r\n\t\tfloat3 pX = x + float3(offset, 0, 0);\r\n\t\tfloat3 pY = x + float3(0, offset, 0);\r\n\t\tfloat3 pZ = x + float3(0, 0, offset);\r\n\t\t\r\n\t\tfloat dX = voronoiDistance(pX);\r\n\t\tfloat dY = voronoiDistance(pY);\r\n\t\tfloat dZ = voronoiDistance(pZ);\r\n\t\t\r\n\t\tfloat edgeX = smoothstep(ss1,ss2,dX);\r\n\t\tfloat edgeY = smoothstep(ss1,ss2,dY);\r\n\t\tfloat edgeZ = smoothstep(ss1,ss2,dZ);\r\n\t\t\r\n\t\tfloat3 intX = float3((edgeX * intensity), 0, 0);\r\n\t\tfloat3 intY = float3(0, (edgeY * intensity), 0);\r\n\t\tfloat3 intZ = float3(0, 0, (edgeZ * intensity));\r\n\t\t\r\n\t\treturn intX + intY + intZ;\r\n\t}\t\r\n};\r\n\r\n\r\n\r\nFunctions3 f3;\r\n\r\nVoronoiCells =  f3.voronoi(p);\r\n\r\nfloat d = f3.voronoiDistance( p );\r\n\r\n\r\nVoronoiEdges = 1.0 - smoothstep(ss1,ss2,d);\r\n\r\nVoronoiNormalOffsets = f3.voronoiNormals(nrm_ss1, nrm_ss2, p, nrmOffset, nrmIntensity);\r\n\r\nreturn 0;\r\n\r\n"
   description=Voronoi
   additionalOutputs [
    {outputName=VoronoiCells, outputType=CMOT_Float4}
    {outputName=VoronoiEdges, outputType=CMOT_Float1}
    {outputName=VoronoiNormalOffsets, outputType=CMOT_Float3}
   ]
   inputs [
    {
     inputName=p
     input {expression=&self:MaterialExpressionMultiply_49}
    }
    {
     inputName=ss1
     input {expression=&self:MaterialExpressionAdd_29}
    }
    {
     inputName=ss2
     input {expression=&self:MaterialExpressionScalarParameter_54}
    }
    {
     inputName=nrmOffset
     input {expression=&self:MaterialExpressionScalarParameter_56}
    }
    {
     inputName=nrmIntensity
     input {expression=&self:MaterialExpressionScalarParameter_55}
    }
    {
     inputName=nrm_ss1
     input {expression=&self:MaterialExpressionAdd_29}
    }
    {
     inputName=nrm_ss2
     input {expression=&self:MaterialExpressionScalarParameter_54}
    }
   ]
  }
  n187 = ComponentMask(None=@n186.VoronoiCells) {b=true, g=true, r=true}
  n188 = Normalize(VectorInput=Subtract(A=Multiply(A=@n187) {constB=2}))
  n189 = LinearInterpolate(
   A=LinearInterpolate(
    A=ScalarParameter() {defaultValue=0.2619, parameterName=DetailVoronoiCellsNormalReorient}
    B=ScalarParameter() {defaultValue=0.2619, parameterName=DetailVoronoiCellsNormalReorientTattooss}
    Alpha=NamedRerouteUsage()
   )
   B=ScalarParameter() {defaultValue=0.2619, parameterName=DetailVoronoiCellsNormalReorientPants}
   Alpha=NamedRerouteUsage()
  )
  n190 = Add(A=Add(A=Transform(None=@n182), B=Multiply(A=@n188, B=@n189)), B=Multiply(A=@n182, B=@n186.VoronoiNormalOffsets)) {constB=0}
  NamedRerouteDeclaration(None=Transform(None=Normalize(VectorInput=@n190)) {transformSourceType=TRANSFORMSOURCE_World, transformType=TRANSFORM_Tangent}) {name=NormalMapWithShards, nodeColor=<rgba 1 0 0.9844 1>}
  NamedRerouteDeclaration(None=@n186.VoronoiEdges) {name=ShardsColor, nodeColor=<rgba 1 0.7266 0 1>}
  n191 = LinearInterpolate(A=@n4, B=Add(A=Multiply(A=@n1, B=@n4), B=@n4), Alpha=@n2)
  NamedRerouteDeclaration(None=Transform(None=Normalize(VectorInput=@n191)) {transformSourceType=TRANSFORMSOURCE_World, transformType=TRANSFORM_Tangent}) {name=NormalsOut, nodeColor=<rgba 1 0.6563 0 1>}
  NamedRerouteUsage()
  OneMinus(None=NamedRerouteUsage())
  Reroute(None=PreSkinnedLocalBounds().Extents)
  VertexNormalWS()
 ]
}
